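# Master particle material for stylized fireball and smoke effects, built on panning world-aligned cloud textures and normals.
# Emissive color pairs dim fake-lit diffuse with glow from color, Fresnel or curve atlas, using a sharp directional or sphere-mask shadow edge.
# Optional fake lighting adds noise-broken Fresnel shading from sun/moon or set direction.
# Opacity uses view-facing billow HeightLerp or custom panning texture; particle dynamic parameters expose size, edge, panning and bloom.
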Material /DealershipSharedLibrary/FX/Materials/MasterMaterials/Smoke/M_DLR_Fireball
graph {
 expressions {
  MaterialExpressionAdd_0 {
   class=MaterialExpressionAdd
   inputs [
    A {from=MaterialExpressionClamp_0}
    B {from=MaterialExpressionClamp_2}
   ]
  }
  MaterialExpressionAdd_1 {
   class=MaterialExpressionAdd
   inputs [
    A {from=MaterialExpressionMultiply_24}
    B {from=MaterialExpressionMultiply_23}
   ]
  }
  MaterialExpressionAdd_2 {
   class=MaterialExpressionAdd
   inputs [
    A {from=MaterialExpressionMultiply_5}
    B {from=MaterialExpressionWorldPosition_0, output=XYZ}
   ]
  }
  MaterialExpressionAdd_3 {
   class=MaterialExpressionAdd
   inputs [
    A {from=MaterialExpressionIf_6}
    B {from=MaterialExpressionNamedRerouteUsage_22}
   ]
   properties {constB=-0.4, desc="1-0 (line)"}
  }
  MaterialExpressionAdd_4 {
   class=MaterialExpressionAdd
   inputs [
    A {from=MaterialExpressionMultiply_7}
    B {from=MaterialExpressionDotProduct_0}
   ]
  }
  MaterialExpressionAdd_5 {
   class=MaterialExpressionAdd
   inputs [
    A {from=MaterialExpressionReroute_6}
    B {from=MaterialExpressionLinearInterpolate_0}
   ]
  }
  MaterialExpressionAdd_6 {
   class=MaterialExpressionAdd
   inputs [
    A {from=MaterialExpressionDotProduct_6}
    B {from=""}
   ]
   properties {constB=-0.1}
  }
  MaterialExpressionAdd_7 {
   class=MaterialExpressionAdd
   inputs [
    A {from=MaterialExpressionNamedRerouteUsage_4}
    B {from=MaterialExpressionIf_1}
   ]
  }
  MaterialExpressionCameraVectorWS_0 {class=MaterialExpressionCameraVectorWS, inputs [], properties {}}
  MaterialExpressionCameraVectorWS_1 {class=MaterialExpressionCameraVectorWS, inputs [], properties {}}
  MaterialExpressionCameraVectorWS_2 {class=MaterialExpressionCameraVectorWS, inputs [], properties {}}
  MaterialExpressionCameraVectorWS_3 {class=MaterialExpressionCameraVectorWS, inputs [], properties {}}
  MaterialExpressionClamp_0 {
   class=MaterialExpressionClamp
   inputs [
    None {from=MaterialExpressionDotProduct_2, output=""}
    Min {from="", output=""}
    Max {from="", output=""}
   ]
   properties {}
  }
  MaterialExpressionClamp_1 {
   class=MaterialExpressionClamp
   inputs [
    None {from=MaterialExpressionMaterialFunctionCall_8, output=Fresnel}
    Min {from="", output=""}
    Max {from="", output=""}
   ]
   properties {}
  }
  MaterialExpressionClamp_2 {
   class=MaterialExpressionClamp
   inputs [
    None {from=MaterialExpressionDotProduct_4, output=""}
    Min {from="", output=""}
    Max {from="", output=""}
   ]
   properties {}
  }
  MaterialExpressionClamp_3 {
   class=MaterialExpressionClamp
   inputs [
    None {from=MaterialExpressionMultiply_6, output=""}
    Min {from="", output=""}
    Max {from="", output=""}
   ]
   properties {}
  }
  MaterialExpressionClamp_4 {
   class=MaterialExpressionClamp
   inputs [
    None {from=MaterialExpressionMultiply_11, output=""}
    Min {from="", output=""}
    Max {from="", output=""}
   ]
   properties {}
  }
  MaterialExpressionClamp_7 {
   class=MaterialExpressionClamp
   inputs [
    None {from=MaterialExpressionNamedRerouteUsage_3, output=""}
    Min {from="", output=""}
    Max {from="", output=""}
   ]
   properties {}
  }
  MaterialExpressionClamp_8 {
   class=MaterialExpressionClamp
   inputs [
    None {from=MaterialExpressionSubtract_2, output=""}
    Min {from="", output=""}
    Max {from="", output=""}
   ]
   properties {}
  }
  MaterialExpressionCollectionParameter_1 {
   class=MaterialExpressionCollectionParameter
   inputs []
   properties {
    collection=&FortniteMaterialParameters
    group=Fake Lighting
    parameterName=SunAndMoonModelDirectionalVector
   }
  }
  MaterialExpressionComponentMask_0 {class=MaterialExpressionComponentMask, inputs [None {from=MaterialExpressionVectorParameter_7}], properties {g=true, r=true}}
  MaterialExpressionComponentMask_1 {class=MaterialExpressionComponentMask, inputs [None {from=MaterialExpressionCollectionParameter_1}], properties {b=true, g=true, r=true}}
  MaterialExpressionConstant3Vector_0 {class=MaterialExpressionConstant3Vector, properties {constant=<rgba 500 500 500 0>}}
  MaterialExpressionConstant3Vector_11 {class=MaterialExpressionConstant3Vector, properties {desc=Angle of light, constant=<rgba 0.2 1.2 0 0>}}
  MaterialExpressionConstant3Vector_6 {class=MaterialExpressionConstant3Vector, properties {constant=<rgba 0 0 1 0>}}
  MaterialExpressionConstant3Vector_8 {class=MaterialExpressionConstant3Vector, properties {constant=<rgba 0 0 300 0>}}
  MaterialExpressionConstant_1 {class=MaterialExpressionConstant}
  MaterialExpressionConstant_2 {class=MaterialExpressionConstant, properties {r=1}}
  MaterialExpressionConstant_3 {class=MaterialExpressionConstant, properties {desc=.2, r=0.2}}
  MaterialExpressionConstant_4 {class=MaterialExpressionConstant}
  MaterialExpressionConvert_1 {
   class=MaterialExpressionConvert
   inputs [
    None {from=MaterialExpressionVectorParameter_7, output=Panner X}
    None {from=MaterialExpressionVectorParameter_7, output=Panner X}
   ]
   properties {
    nodeName=Make Vector2
    convertInputs [
     {
      type=Scalar
      expressionInput {
       expression=&self:MaterialExpressionVectorParameter_7
       inputName=None
       mask=1
       maskA=0
       maskB=1
       maskG=0
       maskR=0
       outputIndex=3
      }
     }
     {
      type=Scalar
      expressionInput {
       expression=&self:MaterialExpressionVectorParameter_7
       inputName=None
       mask=1
       maskA=1
       maskB=0
       maskG=0
       maskR=0
       outputIndex=4
      }
     }
    ]
    convertOutputs [{type=Vector2}]
   }
  }
  MaterialExpressionCurveAtlasRowParameter_0 {
   class=MaterialExpressionCurveAtlasRowParameter
   inputs [CurveTime {from=MaterialExpressionFresnel_0, output=""}]
   properties {
    atlas=&AthenaCurveAtlas01
    curve=/Game/Gadgets/Assets/VinderTech_GliderChute/Glider_Valentines/FX/ValentinesGlider_Feathers
    defaultValue=57
    parameterName=Curve
   }
  }
  MaterialExpressionDivide_0 {
   class=MaterialExpressionDivide
   inputs [
    A {from=MaterialExpressionAdd_0}
    B {from=MaterialExpressionScalarParameter_6}
   ]
  }
  MaterialExpressionDotProduct_0 {
   class=MaterialExpressionDotProduct
   inputs [
    A {from=MaterialExpressionNamedRerouteUsage_15, output=""}
    B {from=MaterialExpressionReroute_15, output=""}
   ]
   properties {}
  }
  MaterialExpressionDotProduct_2 {
   class=MaterialExpressionDotProduct
   inputs [
    A {from=MaterialExpressionNamedRerouteUsage_14, output=""}
    B {from=MaterialExpressionCameraVectorWS_0, output=""}
   ]
   properties {}
  }
  MaterialExpressionDotProduct_3 {
   class=MaterialExpressionDotProduct
   inputs [
    A {from=MaterialExpressionTextureSampleParameter2D_3, output=RGBA}
    B {from=MaterialExpressionVectorParameter_8, output=RGBA}
   ]
   properties {}
  }
  MaterialExpressionDotProduct_4 {
   class=MaterialExpressionDotProduct
   inputs [
    A {from=MaterialExpressionVertexNormalWS_0, output=""}
    B {from=MaterialExpressionCameraVectorWS_2, output=""}
   ]
   properties {}
  }
  MaterialExpressionDotProduct_6 {
   class=MaterialExpressionDotProduct
   inputs [
    A {from=MaterialExpressionNamedRerouteUsage_16, output=""}
    B {from=MaterialExpressionTransform_0, output=""}
   ]
   properties {}
  }
  MaterialExpressionDynamicParameter_0 {
   class=MaterialExpressionDynamicParameter
   inputs []
   properties {defaultValue=<rgba 0 0 0 1>, paramNames [Texture Size, ControlShadowEdge, PanTexture, ControlBloom]}
  }
  MaterialExpressionFresnel_0 {
   class=MaterialExpressionFresnel
   inputs [
    ExponentIn {from=MaterialExpressionScalarParameter_4, output=""}
    BaseReflectFractionIn {from="", output=""}
    Normal {from="", output=""}
   ]
   properties {}
  }
  MaterialExpressionFresnel_1 {
   class=MaterialExpressionFresnel
   inputs [
    ExponentIn {from="", output=""}
    BaseReflectFractionIn {from="", output=""}
    Normal {from="", output=""}
   ]
   properties {}
  }
  MaterialExpressionIf_0 {
   class=MaterialExpressionIf
   inputs [
    A {from=MaterialExpressionScalarParameter_5, output=""}
    B {from="", output=""}
    "A > B" {from=MaterialExpressionNamedRerouteUsage_5, output=""}
    "A == B" {from=MaterialExpressionNamedRerouteUsage_5, output=""}
    "A < B" {from=MaterialExpressionReroute_10, output=""}
   ]
   properties {constB=0.5}
  }
  MaterialExpressionIf_1 {
   class=MaterialExpressionIf
   inputs [
    A {from=MaterialExpressionScalarParameter_13, output=""}
    B {from="", output=""}
    "A > B" {from=MaterialExpressionIf_4, output=""}
    "A == B" {from=MaterialExpressionIf_4, output=""}
    "A < B" {from=MaterialExpressionVectorParameter_4, output=RGB}
   ]
   properties {constB=1}
  }
  MaterialExpressionIf_10 {
   class=MaterialExpressionIf
   inputs [
    A {from=MaterialExpressionScalarParameter_14, output=""}
    B {from="", output=""}
    "A > B" {from=MaterialExpressionNamedRerouteUsage_6, output=""}
    "A == B" {from=MaterialExpressionLinearInterpolate_5, output=""}
    "A < B" {from=MaterialExpressionLinearInterpolate_5, output=""}
   ]
   properties {}
  }
  MaterialExpressionIf_2 {
   class=MaterialExpressionIf
   inputs [
    A {from=MaterialExpressionScalarParameter_26, output=""}
    B {from="", output=""}
    "A > B" {from=MaterialExpressionCurveAtlasRowParameter_0, output=""}
    "A == B" {from=MaterialExpressionCurveAtlasRowParameter_0, output=""}
    "A < B" {from=MaterialExpressionReroute_1, output=""}
   ]
   properties {constB=1}
  }
  MaterialExpressionIf_3 {
   class=MaterialExpressionIf
   inputs [
    A {from=MaterialExpressionScalarParameter_27, output=""}
    B {from="", output=""}
    "A > B" {from=MaterialExpressionReroute_2, output=""}
    "A == B" {from=MaterialExpressionReroute_3, output=""}
    "A < B" {from=MaterialExpressionMaterialFunctionCall_5, output=Result}
   ]
   properties {constB=1}
  }
  MaterialExpressionIf_4 {
   class=MaterialExpressionIf
   inputs [
    A {from=MaterialExpressionScalarParameter_28, output=""}
    B {from="", output=""}
    "A > B" {from=MaterialExpressionLinearInterpolate_9, output=""}
    "A == B" {from=MaterialExpressionReroute_9, output=""}
    "A < B" {from=MaterialExpressionReroute_7, output=""}
   ]
   properties {constB=1}
  }
  MaterialExpressionIf_5 {
   class=MaterialExpressionIf
   inputs [
    A {from=MaterialExpressionScalarParameter_29, output=""}
    B {from="", output=""}
    "A > B" {from=MaterialExpressionNormalize_0, output=""}
    "A == B" {from=MaterialExpressionNormalize_0, output=""}
    "A < B" {from=MaterialExpressionReroute_16, output=""}
   ]
   properties {constB=1}
  }
  MaterialExpressionIf_6 {
   class=MaterialExpressionIf
   inputs [
    A {from=MaterialExpressionScalarParameter_30, output=""}
    B {from="", output=""}
    "A > B" {from=MaterialExpressionReroute_17, output=""}
    "A == B" {from=MaterialExpressionReroute_18, output=""}
    "A < B" {from=MaterialExpressionAdd_4, output=""}
   ]
   properties {constB=1}
  }
  MaterialExpressionIf_7 {
   class=MaterialExpressionIf
   inputs [
    A {from=MaterialExpressionScalarParameter_31, output=""}
    B {from="", output=""}
    "A > B" {from=MaterialExpressionMultiply_0, output=""}
    "A == B" {from=MaterialExpressionMultiply_0, output=""}
    "A < B" {from=MaterialExpressionNamedRerouteUsage_12, output=""}
   ]
   properties {constB=1}
  }
  MaterialExpressionIf_8 {
   class=MaterialExpressionIf
   inputs [
    A {from=MaterialExpressionScalarParameter_32, output=""}
    B {from="", output=""}
    "A > B" {from=MaterialExpressionMultiply_3, output=""}
    "A == B" {from=MaterialExpressionMultiply_3, output=""}
    "A < B" {from=MaterialExpressionMaterialFunctionCall_4, output=Alpha}
   ]
   properties {constB=1}
  }
  MaterialExpressionIf_9 {
   class=MaterialExpressionIf
   inputs [
    A {from=MaterialExpressionScalarParameter_34, output=""}
    B {from="", output=""}
    "A > B" {from=MaterialExpressionTransform_1, output=""}
    "A == B" {from=MaterialExpressionTransform_1, output=""}
    "A < B" {from=MaterialExpressionCameraVectorWS_1, output=""}
   ]
   properties {constB=1}
  }
  MaterialExpressionLinearInterpolate_0 {
   class=MaterialExpressionLinearInterpolate
   inputs [
    A {from=MaterialExpressionNamedRerouteUsage_13}
    B {from=MaterialExpressionMultiply_15}
    Alpha {from=MaterialExpressionNamedRerouteUsage_17}
   ]
  }
  MaterialExpressionLinearInterpolate_1 {
   class=MaterialExpressionLinearInterpolate
   inputs [
    A {from=MaterialExpressionMaterialFunctionCall_2, output=Normal XYZ Texture}
    B {from=MaterialExpressionVertexNormalWS_1}
    Alpha {from=MaterialExpressionConstant_3}
   ]
   properties {desc=Soft normal}
  }
  MaterialExpressionLinearInterpolate_2 {
   class=MaterialExpressionLinearInterpolate
   inputs [
    A {from=""}
    B {from=""}
    Alpha {from=MaterialExpressionScalarParameter_2}
   ]
   properties {constB=40}
  }
  MaterialExpressionLinearInterpolate_3 {
   class=MaterialExpressionLinearInterpolate
   inputs [
    A {from=MaterialExpressionVectorParameter_3, output=RGB}
    B {from=MaterialExpressionMultiply_8}
    Alpha {from=MaterialExpressionScalarParameter_8}
   ]
  }
  MaterialExpressionLinearInterpolate_4 {
   class=MaterialExpressionLinearInterpolate
   inputs [
    A {from=MaterialExpressionReroute_13}
    B {from=MaterialExpressionAdd_2}
    Alpha {from=MaterialExpressionScalarParameter_9}
   ]
  }
  MaterialExpressionLinearInterpolate_5 {
   class=MaterialExpressionLinearInterpolate
   inputs [
    A {from=MaterialExpressionNamedRerouteUsage_6}
    B {from=MaterialExpressionVectorParameter_5, output=RGB}
    Alpha {from=MaterialExpressionClamp_7}
   ]
  }
  MaterialExpressionLinearInterpolate_6 {
   class=MaterialExpressionLinearInterpolate
   inputs [
    A {from=MaterialExpressionConstant_1}
    B {from=MaterialExpressionMaterialFunctionCall_6, output=XYZ Texture}
    Alpha {from=MaterialExpressionScalarParameter_20}
   ]
   properties {constB=0}
  }
  MaterialExpressionLinearInterpolate_9 {
   class=MaterialExpressionLinearInterpolate
   inputs [
    A {from=MaterialExpressionComponentMask_1}
    B {from=MaterialExpressionIf_9}
    Alpha {from=MaterialExpressionReroute_11}
   ]
  }
  MaterialExpressionMaterialFunctionCall_0 {
   class=MaterialExpressionMaterialFunctionCall
   inputs [Distance {from=MaterialExpressionScalarParameter_3}]
   properties {materialFunction=&Near_Camera_Fade}
  }
  MaterialExpressionMaterialFunctionCall_1 {
   class=MaterialExpressionMaterialFunctionCall
   inputs [
    Brightness {from=""}
    Exponent {from=MaterialExpressionScalarParameter_7}
    Camera Vector {from=""}
    Vertex / Pixel Normal {from=""}
   ]
   properties {materialFunction=&MF_Fresnel_FX}
  }
  MaterialExpressionMaterialFunctionCall_11 {
   class=MaterialExpressionMaterialFunctionCall
   inputs [
    DayBloom_Mult {from=MaterialExpressionConstant_2}
    In {from=MaterialExpressionIf_2}
   ]
   properties {materialFunction=/Game/Packages/Fortnite_UniqueMaterials_1/MH_Materials/DayNight_Auto_Bloom_Control}
  }
  MaterialExpressionMaterialFunctionCall_2 {
   class=MaterialExpressionMaterialFunctionCall
   inputs [
    Diffuse Texture Object {from=MaterialExpressionTextureObjectParameter_1}
    Spec Texture Object {from=""}
    TextureSize {from=MaterialExpressionMultiply_4}
    -------------------------- {from=""}
    Use seperate settings for Z projection {from=""}
    Use seperate texture for Z projection {from=""}
    Texture Size Z projection {from=""}
    Z Plane Diffuse Texture Object {from=""}
    Z Plane Spec Texture Object {from=""}
    "--------------------------      " {from=""}
    ProjectionTransitionContrast {from=""}
    Use Texture Alpha Heightmap to Refine the Transistion {from=""}
    HeightLerpTransitionContrast {from=""}
    PickTheDiffuseHeightChannel {from=""}
    TheDiffuseHeightChannel {from=""}
    "--------------------------       " {from=""}
    NormalTextureObject {from=MaterialExpressionTextureObjectParameter_2}
    World Space Normals {from=""}
    "--------------------------  " {from=""}
    Export Float 4 Diffuse {from=""}
    Export Float 4 Spec {from=""}
    "--------------------------     " {from=""}
    WorldPosition {from=MaterialExpressionLinearInterpolate_4}
    World Space Normal {from=""}
   ]
   properties {materialFunction=/Engine/Functions/Engine_MaterialFunctions02/Texturing/WorldAlignedTextures_Complex}
  }
  MaterialExpressionMaterialFunctionCall_3 {
   class=MaterialExpressionMaterialFunctionCall
   inputs [Emissive {from=MaterialExpressionIf_10}]
   properties {materialFunction=&MF_FX_EmissiveAdjustments}
  }
  MaterialExpressionMaterialFunctionCall_4 {
   class=MaterialExpressionMaterialFunctionCall
   inputs [
    A {from=""}
    B {from=""}
    Transition Phase {from=MaterialExpressionIf_7}
    Height Texture {from=MaterialExpressionReroute_20}
    Contrast {from=""}
   ]
   properties {materialFunction=&HeightLerp}
  }
  MaterialExpressionMaterialFunctionCall_5 {
   class=MaterialExpressionMaterialFunctionCall
   inputs [
    In {from=MaterialExpressionMultiply_1}
    Contrast {from=MaterialExpressionScalarParameter_12}
   ]
   properties {materialFunction=&CheapContrast}
  }
  MaterialExpressionMaterialFunctionCall_6 {
   class=MaterialExpressionMaterialFunctionCall
   inputs [
    Export Float 4 {from=""}
    TextureObject {from=MaterialExpressionTextureObjectParameter_0}
    TextureSize {from=MaterialExpressionReroute_12}
    WorldPosition {from=""}
    World Space Normal {from=""}
    RotZAxis {from=""}
    RotXAxis {from=""}
    RotYAxis {from=""}
    ProjectionTransitionContrast {from=""}
   ]
   properties {materialFunction=/Game/Weapons/WeaponSkins/Wraps/MaterialFunctions/WorldAlignedTexture}
  }
  MaterialExpressionMaterialFunctionCall_8 {
   class=MaterialExpressionMaterialFunctionCall
   inputs [
    Brightness {from=MaterialExpressionScalarParameter_33}
    Exponent {from=MaterialExpressionScalarParameter_36}
    Camera Vector {from=MaterialExpressionAdd_7}
    Vertex / Pixel Normal {from=""}
   ]
   properties {materialFunction=&MF_Fresnel_FX}
  }
  MaterialExpressionMultiply_0 {
   class=MaterialExpressionMultiply
   inputs [
    A {from=MaterialExpressionMaterialFunctionCall_0, output=Result}
    B {from=MaterialExpressionNamedRerouteUsage_10}
   ]
  }
  MaterialExpressionMultiply_1 {
   class=MaterialExpressionMultiply
   inputs [
    A {from=MaterialExpressionMaterialFunctionCall_1, output=Fresnel}
    B {from=MaterialExpressionVectorParameter_2, output=RGB}
   ]
  }
  MaterialExpressionMultiply_10 {
   class=MaterialExpressionMultiply
   inputs [
    A {from=MaterialExpressionClamp_4}
    B {from=MaterialExpressionReroute_4}
   ]
   properties {constB=0.5}
  }
  MaterialExpressionMultiply_11 {
   class=MaterialExpressionMultiply
   inputs [
    A {from=MaterialExpressionAdd_6}
    B {from=MaterialExpressionLinearInterpolate_2}
   ]
   properties {constB=40}
  }
  MaterialExpressionMultiply_12 {
   class=MaterialExpressionMultiply
   inputs [
    A {from=MaterialExpressionMultiply_10}
    B {from=MaterialExpressionScalarParameter_1}
   ]
   properties {constB=0.2}
  }
  MaterialExpressionMultiply_13 {
   class=MaterialExpressionMultiply
   inputs [
    A {from=MaterialExpressionMaterialFunctionCall_2, output=Normal XYZ Texture}
    B {from=""}
   ]
  }
  MaterialExpressionMultiply_14 {
   class=MaterialExpressionMultiply
   inputs [
    A {from=MaterialExpressionNamedRerouteUsage_20}
    B {from=MaterialExpressionConstant3Vector_8}
   ]
  }
  MaterialExpressionMultiply_15 {
   class=MaterialExpressionMultiply
   inputs [
    A {from=MaterialExpressionMaterialFunctionCall_11, output=Result}
    B {from=MaterialExpressionNamedRerouteUsage_19}
   ]
  }
  MaterialExpressionMultiply_2 {
   class=MaterialExpressionMultiply
   inputs [
    A {from=MaterialExpressionVectorParameter_0, output=RGB}
    B {from=MaterialExpressionMaterialFunctionCall_5, output=Result}
   ]
  }
  MaterialExpressionMultiply_20 {
   class=MaterialExpressionMultiply
   inputs [
    A {from=MaterialExpressionVectorParameter_6, output=RGB}
    B {from=MaterialExpressionScalarParameter_18}
   ]
  }
  MaterialExpressionMultiply_23 {
   class=MaterialExpressionMultiply
   inputs [
    A {from=MaterialExpressionConvert_1}
    B {from=MaterialExpressionTime_1}
   ]
  }
  MaterialExpressionMultiply_24 {
   class=MaterialExpressionMultiply
   inputs [
    A {from=MaterialExpressionTextureCoordinate_3}
    B {from=MaterialExpressionComponentMask_0}
   ]
  }
  MaterialExpressionMultiply_3 {
   class=MaterialExpressionMultiply
   inputs [
    A {from=MaterialExpressionNamedRerouteUsage_11}
    B {from=MaterialExpressionMaterialFunctionCall_4, output=Alpha}
   ]
  }
  MaterialExpressionMultiply_4 {
   class=MaterialExpressionMultiply
   inputs [
    A {from=MaterialExpressionConstant3Vector_0}
    B {from=MaterialExpressionNamedRerouteUsage_21}
   ]
   properties {constB=3}
  }
  MaterialExpressionMultiply_5 {
   class=MaterialExpressionMultiply
   inputs [
    A {from=MaterialExpressionTime_0}
    B {from=MaterialExpressionMultiply_14}
   ]
   properties {constB=200}
  }
  MaterialExpressionMultiply_6 {
   class=MaterialExpressionMultiply
   inputs [
    A {from=MaterialExpressionAdd_3}
    B {from=""}
   ]
   properties {constB=15, desc=sharp line}
  }
  MaterialExpressionMultiply_7 {
   class=MaterialExpressionMultiply
   inputs [
    A {from=MaterialExpressionFresnel_1}
    B {from=MaterialExpressionScalarParameter_11}
   ]
   properties {constB=10}
  }
  MaterialExpressionMultiply_8 {
   class=MaterialExpressionMultiply
   inputs [
    A {from=MaterialExpressionCameraVectorWS_3}
    B {from=""}
   ]
   properties {constB=-1}
  }
  MaterialExpressionNamedRerouteDeclaration_0 {
   class=MaterialExpressionNamedRerouteDeclaration
   inputs [None {from=MaterialExpressionAdd_5}]
   properties {name=Emissive Texture, nodeColor=<rgba 1 0 0.0703 1>}
  }
  MaterialExpressionNamedRerouteDeclaration_1 {
   class=MaterialExpressionNamedRerouteDeclaration
   inputs [None {from=MaterialExpressionClamp_8}]
   properties {name=Fake Lighting, nodeColor=<rgba 1 0 0.5625 1>}
  }
  MaterialExpressionNamedRerouteDeclaration_10 {
   class=MaterialExpressionNamedRerouteDeclaration
   inputs [None {from=MaterialExpressionClamp_3}]
   properties {name=Shadow Edge, nodeColor=<rgba 1 0.3984 0 1>}
  }
  MaterialExpressionNamedRerouteDeclaration_11 {
   class=MaterialExpressionNamedRerouteDeclaration
   inputs [None {from=MaterialExpressionVectorParameter_1, output=RGB}]
   properties {name=Diffuse Color, nodeColor=<rgba 0.0781 0 1 1>}
  }
  MaterialExpressionNamedRerouteDeclaration_12 {
   class=MaterialExpressionNamedRerouteDeclaration
   inputs [None {from=MaterialExpressionDynamicParameter_0, output=Texture Size}]
   properties {name=DP_Texture Size, nodeColor=<rgba 0.2422 1 0 1>}
  }
  MaterialExpressionNamedRerouteDeclaration_13 {
   class=MaterialExpressionNamedRerouteDeclaration
   inputs [None {from=MaterialExpressionDynamicParameter_0, output=ControlShadowEdge}]
   properties {name=DP_ControlShadowEdge, nodeColor=<rgba 0.2422 1 0 1>}
  }
  MaterialExpressionNamedRerouteDeclaration_14 {
   class=MaterialExpressionNamedRerouteDeclaration
   inputs [None {from=MaterialExpressionDynamicParameter_0, output=PanTexture}]
   properties {name=DP_PanTexture, nodeColor=<rgba 0.2422 1 0 1>}
  }
  MaterialExpressionNamedRerouteDeclaration_15 {
   class=MaterialExpressionNamedRerouteDeclaration
   inputs [None {from=MaterialExpressionDynamicParameter_0, output=ControlBloom}]
   properties {name=DP_ControlBloom, nodeColor=<rgba 0.2422 1 0 1>}
  }
  MaterialExpressionNamedRerouteDeclaration_2 {
   class=MaterialExpressionNamedRerouteDeclaration
   inputs [None {from=MaterialExpressionSubtract_1}]
   properties {name=CustomOpacity, nodeColor=<rgba 1 0 0.75 1>}
  }
  MaterialExpressionNamedRerouteDeclaration_3 {
   class=MaterialExpressionNamedRerouteDeclaration
   inputs [None {from=MaterialExpressionLinearInterpolate_6}]
   properties {name=Fake Lighting Noise, nodeColor=<rgba 1 0 0.1641 1>}
  }
  MaterialExpressionNamedRerouteDeclaration_6 {
   class=MaterialExpressionNamedRerouteDeclaration
   inputs [None {from=MaterialExpressionIf_0}]
   properties {name=Opacity, nodeColor=<rgba 0.9688 1 0 1>}
  }
  MaterialExpressionNamedRerouteDeclaration_7 {
   class=MaterialExpressionNamedRerouteDeclaration
   inputs [None {from=MaterialExpressionParticleColor_0, output=A}]
   properties {name=Particle Alpha, nodeColor=<rgba 0.6406 1 0 1>}
  }
  MaterialExpressionNamedRerouteDeclaration_8 {
   class=MaterialExpressionNamedRerouteDeclaration
   inputs [None {from=MaterialExpressionParticleColor_0, output=RGB}]
   properties {name=Particle Color, nodeColor=<rgba 1 0 0.9609 1>}
  }
  MaterialExpressionNamedRerouteDeclaration_9 {
   class=MaterialExpressionNamedRerouteDeclaration
   inputs [None {from=MaterialExpressionLinearInterpolate_1}]
   properties {name=Normal XYZ Texture, nodeColor=<rgba 0 1 0.7188 1>}
  }
  MaterialExpressionNamedRerouteUsage_1 {class=MaterialExpressionNamedRerouteUsage}
  MaterialExpressionNamedRerouteUsage_10 {class=MaterialExpressionNamedRerouteUsage}
  MaterialExpressionNamedRerouteUsage_11 {class=MaterialExpressionNamedRerouteUsage}
  MaterialExpressionNamedRerouteUsage_12 {class=MaterialExpressionNamedRerouteUsage}
  MaterialExpressionNamedRerouteUsage_13 {class=MaterialExpressionNamedRerouteUsage}
  MaterialExpressionNamedRerouteUsage_14 {class=MaterialExpressionNamedRerouteUsage}
  MaterialExpressionNamedRerouteUsage_15 {class=MaterialExpressionNamedRerouteUsage}
  MaterialExpressionNamedRerouteUsage_16 {class=MaterialExpressionNamedRerouteUsage}
  MaterialExpressionNamedRerouteUsage_17 {class=MaterialExpressionNamedRerouteUsage}
  MaterialExpressionNamedRerouteUsage_18 {class=MaterialExpressionNamedRerouteUsage}
  MaterialExpressionNamedRerouteUsage_19 {class=MaterialExpressionNamedRerouteUsage}
  MaterialExpressionNamedRerouteUsage_2 {class=MaterialExpressionNamedRerouteUsage}
  MaterialExpressionNamedRerouteUsage_20 {class=MaterialExpressionNamedRerouteUsage}
  MaterialExpressionNamedRerouteUsage_21 {class=MaterialExpressionNamedRerouteUsage}
  MaterialExpressionNamedRerouteUsage_22 {class=MaterialExpressionNamedRerouteUsage}
  MaterialExpressionNamedRerouteUsage_3 {class=MaterialExpressionNamedRerouteUsage}
  MaterialExpressionNamedRerouteUsage_4 {class=MaterialExpressionNamedRerouteUsage}
  MaterialExpressionNamedRerouteUsage_5 {class=MaterialExpressionNamedRerouteUsage}
  MaterialExpressionNamedRerouteUsage_6 {class=MaterialExpressionNamedRerouteUsage}
  MaterialExpressionNamedRerouteUsage_8 {class=MaterialExpressionNamedRerouteUsage}
  MaterialExpressionNamedRerouteUsage_9 {class=MaterialExpressionNamedRerouteUsage}
  MaterialExpressionNormalize_0 {class=MaterialExpressionNormalize, inputs [VectorInput {from=MaterialExpressionSubtract_0, output=""}]}
  MaterialExpressionObjectPositionWS_0 {class=MaterialExpressionObjectPositionWS, inputs [], properties {}}
  MaterialExpressionObjectPositionWS_1 {class=MaterialExpressionObjectPositionWS, inputs [], properties {}}
  MaterialExpressionParticleColor_0 {class=MaterialExpressionParticleColor, inputs [], properties {}}
  MaterialExpressionPower_0 {
   class=MaterialExpressionPower
   inputs [
    Base {from=MaterialExpressionSphereMask_0}
    Exp {from=MaterialExpressionScalarParameter_0}
   ]
  }
  MaterialExpressionReroute_1 {class=MaterialExpressionReroute, inputs [None {from=MaterialExpressionIf_3}]}
  MaterialExpressionReroute_10 {class=MaterialExpressionReroute, inputs [None {from=MaterialExpressionIf_8}]}
  MaterialExpressionReroute_11 {class=MaterialExpressionReroute, inputs [None {from=MaterialExpressionScalarParameter_19}]}
  MaterialExpressionReroute_12 {class=MaterialExpressionReroute, inputs [None {from=MaterialExpressionMultiply_20}]}
  MaterialExpressionReroute_13 {class=MaterialExpressionReroute, inputs [None {from=MaterialExpressionReroute_14}]}
  MaterialExpressionReroute_14 {class=MaterialExpressionReroute, inputs [None {from=MaterialExpressionMultiply_5}]}
  MaterialExpressionReroute_15 {class=MaterialExpressionReroute, inputs [None {from=MaterialExpressionIf_5}]}
  MaterialExpressionReroute_16 {class=MaterialExpressionReroute, inputs [None {from=MaterialExpressionLinearInterpolate_3}]}
  MaterialExpressionReroute_17 {class=MaterialExpressionReroute, inputs [None {from=MaterialExpressionReroute_19}]}
  MaterialExpressionReroute_18 {class=MaterialExpressionReroute, inputs [None {from=MaterialExpressionReroute_19}]}
  MaterialExpressionReroute_19 {class=MaterialExpressionReroute, inputs [None {from=MaterialExpressionPower_0}]}
  MaterialExpressionReroute_2 {class=MaterialExpressionReroute, inputs [None {from=MaterialExpressionMultiply_2}]}
  MaterialExpressionReroute_20 {class=MaterialExpressionReroute, inputs [None {from=MaterialExpressionDivide_0}]}
  MaterialExpressionReroute_3 {class=MaterialExpressionReroute, inputs [None {from=MaterialExpressionMultiply_2}]}
  MaterialExpressionReroute_4 {class=MaterialExpressionReroute, inputs [None {from=MaterialExpressionVectorParameter_1, output=RGB}]}
  MaterialExpressionReroute_5 {class=MaterialExpressionReroute, inputs [None {from=MaterialExpressionMultiply_12}]}
  MaterialExpressionReroute_6 {class=MaterialExpressionReroute, inputs [None {from=MaterialExpressionReroute_5}]}
  MaterialExpressionReroute_7 {class=MaterialExpressionReroute, inputs [None {from=MaterialExpressionReroute_8}]}
  MaterialExpressionReroute_8 {class=MaterialExpressionReroute, inputs [None {from=MaterialExpressionIf_9}]}
  MaterialExpressionReroute_9 {class=MaterialExpressionReroute, inputs [None {from=MaterialExpressionLinearInterpolate_9}]}
  MaterialExpressionScalarParameter_0 {class=MaterialExpressionScalarParameter, properties {defaultValue=4, group=Fake Shadows, parameterName=Fake Shadows - FalloffPower}}
  MaterialExpressionScalarParameter_1 {class=MaterialExpressionScalarParameter, properties {defaultValue=0.2, group=Emissive, parameterName=Emissive - DiffuseBrightness}}
  MaterialExpressionScalarParameter_10 {class=MaterialExpressionScalarParameter, properties {defaultValue=145.6266, group=Fake Shadows, parameterName=Fake Shadows - GlowRadius}}
  MaterialExpressionScalarParameter_11 {class=MaterialExpressionScalarParameter, properties {defaultValue=10, group=Fake Shadows, parameterName=Fake Shadows - Fresnel Brightness}}
  MaterialExpressionScalarParameter_12 {class=MaterialExpressionScalarParameter, properties {group=Emissive, parameterName=Emissive - FresnelGlowContrast}}
  MaterialExpressionScalarParameter_13 {
   class=MaterialExpressionScalarParameter
   properties {defaultValue=1, group=Fake Lighting, parameterName=Fake Lighting - Use TOD vs Set Direction, sortPriority=2}
  }
  MaterialExpressionScalarParameter_14 {class=MaterialExpressionScalarParameter, properties {parameterName=No Fake Lighting}}
  MaterialExpressionScalarParameter_15 {class=MaterialExpressionScalarParameter, properties {group=Fake Shadows, parameterName=Fake Shadows - GlowHardness}}
  MaterialExpressionScalarParameter_18 {
   class=MaterialExpressionScalarParameter
   properties {defaultValue=1, group=Fake Lighting, parameterName=Fake Lighting - Noise Projection Tiling}
  }
  MaterialExpressionScalarParameter_19 {
   class=MaterialExpressionScalarParameter
   properties {defaultValue=0.5, group=Fake Lighting, parameterName=Fake Lighting - Sun/Moon Directional Shading Blend Smoke}
  }
  MaterialExpressionScalarParameter_2 {class=MaterialExpressionScalarParameter, properties {defaultValue=1, group=Emissive, parameterName=Emissive - DiffuseHardness}}
  MaterialExpressionScalarParameter_20 {class=MaterialExpressionScalarParameter, properties {defaultValue=1, group=Fake Lighting, parameterName=Fake Lighting - Noise Amount}}
  MaterialExpressionScalarParameter_25 {class=MaterialExpressionScalarParameter, properties {parameterName=Specular Amount}}
  MaterialExpressionScalarParameter_26 {class=MaterialExpressionScalarParameter, properties {group=Emissive, parameterName=Emissive - Use Curve Asset for Glow Color?}}
  MaterialExpressionScalarParameter_27 {class=MaterialExpressionScalarParameter, properties {group=Emissive, parameterName=Emissive - Use Fresnel For Glow?}}
  MaterialExpressionScalarParameter_28 {class=MaterialExpressionScalarParameter, properties {defaultValue=1, group=Fake Lighting, parameterName=Fake Lighting - ToD Shading?}}
  MaterialExpressionScalarParameter_29 {class=MaterialExpressionScalarParameter, properties {group=Fake Shadows, parameterName=Fake Shadows - Use Object Center for Dot Direction?}}
  MaterialExpressionScalarParameter_3 {class=MaterialExpressionScalarParameter, properties {defaultValue=500, group=Opacity, parameterName=Opacity - Near Distance Fade}}
  MaterialExpressionScalarParameter_30 {class=MaterialExpressionScalarParameter, properties {group=Fake Shadows, parameterName=Fake Shadows - Use Sphere Mask Falloff}}
  MaterialExpressionScalarParameter_31 {class=MaterialExpressionScalarParameter, properties {group=Opacity, parameterName=Opacity - Use Near Camera Fade}}
  MaterialExpressionScalarParameter_32 {class=MaterialExpressionScalarParameter, properties {group=Opacity, parameterName=Opacity - Use Particle Opacity}}
  MaterialExpressionScalarParameter_33 {
   class=MaterialExpressionScalarParameter
   properties {defaultValue=1, group=Fake Lighting, parameterName=Fake Lighting - Main Shading Fresnel Brightness, sortPriority=2}
  }
  MaterialExpressionScalarParameter_34 {class=MaterialExpressionScalarParameter, properties {group=Fake Lighting, parameterName=Fake Lighting - Velocity Aligned Fresnel}}
  MaterialExpressionScalarParameter_36 {
   class=MaterialExpressionScalarParameter
   properties {defaultValue=2, group=Fake Lighting, parameterName=Fake Lighting - Main Shading Fresnel Exponent, sortPriority=2}
  }
  MaterialExpressionScalarParameter_4 {class=MaterialExpressionScalarParameter, properties {defaultValue=5, group=Emissive, parameterName=Emissive - Curve Fresnel Exp}}
  MaterialExpressionScalarParameter_5 {
   class=MaterialExpressionScalarParameter
   properties {
    controlType=Enumeration
    enumeration=/DealershipSharedLibrary/FX/Materials/Enums/Enum_DLR_TrueFalse
    group=CustomOpacity
    parameterName=UseCustomOpacity?
   }
  }
  MaterialExpressionScalarParameter_6 {class=MaterialExpressionScalarParameter, properties {defaultValue=2, group=Opacity, parameterName=Opacity - BillowControl}}
  MaterialExpressionScalarParameter_7 {class=MaterialExpressionScalarParameter, properties {group=Emissive, parameterName=Emissive - FresnelGlowExp}}
  MaterialExpressionScalarParameter_8 {class=MaterialExpressionScalarParameter, properties {group=Fake Shadows, parameterName=Fake Shadows - Use Camera Vector For Glow Direction}}
  MaterialExpressionScalarParameter_9 {class=MaterialExpressionScalarParameter, properties {defaultValue=1, group=World Aligned Texture, parameterName=UsePannerWPO?}}
  MaterialExpressionSphereMask_0 {
   class=MaterialExpressionSphereMask
   inputs [
    A {from=MaterialExpressionWorldPosition_1, output=XYZ}
    B {from=MaterialExpressionObjectPositionWS_0, output=""}
    Radius {from=MaterialExpressionScalarParameter_10, output=""}
    Hardness {from=MaterialExpressionScalarParameter_15, output=""}
   ]
   properties {}
  }
  MaterialExpressionStaticSwitchParameter_7 {
   class=MaterialExpressionStaticSwitchParameter
   inputs [
    True {from=MaterialExpressionClamp_1, output=""}
    False {from=MaterialExpressionConstant_4, output=""}
   ]
   properties {group=Fake Lighting, parameterName=Use Fake Lighting, sortPriority=1}
  }
  MaterialExpressionSubtract_0 {
   class=MaterialExpressionSubtract
   inputs [
    A {from=MaterialExpressionWorldPosition_2, output=XYZ}
    B {from=MaterialExpressionObjectPositionWS_1}
   ]
  }
  MaterialExpressionSubtract_1 {
   class=MaterialExpressionSubtract
   inputs [
    A {from=MaterialExpressionNamedRerouteUsage_1}
    B {from=MaterialExpressionDotProduct_3}
   ]
  }
  MaterialExpressionSubtract_2 {
   class=MaterialExpressionSubtract
   inputs [
    A {from=MaterialExpressionNamedRerouteUsage_2}
    B {from=MaterialExpressionStaticSwitchParameter_7}
   ]
  }
  MaterialExpressionTextureCoordinate_3 {class=MaterialExpressionTextureCoordinate}
  MaterialExpressionTextureObjectParameter_0 {
   class=MaterialExpressionTextureObjectParameter
   inputs [
    UVs {from="", output=""}
    Apply View MipBias {from="", output=""}
   ]
   properties {group=Fake Lighting, parameterName=Fake Lighting - Noise, texture=/Game/Global/Textures/Default/Tileable_Patterns/T_Tornado_Tile}
  }
  MaterialExpressionTextureObjectParameter_1 {
   class=MaterialExpressionTextureObjectParameter
   inputs [
    UVs {from="", output=""}
    Apply View MipBias {from="", output=""}
   ]
   properties {
    group=World Aligned Texture
    parameterName=World Aligned Texture
    samplerType=SAMPLERTYPE_LinearColor
    texture=/Game/Global/Textures/Default/Tileable_Patterns/Cloud_Diffuse
   }
  }
  MaterialExpressionTextureObjectParameter_2 {
   class=MaterialExpressionTextureObjectParameter
   inputs [
    UVs {from="", output=""}
    Apply View MipBias {from="", output=""}
   ]
   properties {
    group=World Aligned Texture
    parameterName=World Aligned Texture - Normals
    samplerType=SAMPLERTYPE_Normal
    texture=/Game/Global/Textures/Default/Tileable_Patterns/Cloud_normals
   }
  }
  MaterialExpressionTextureSampleParameter2D_3 {
   class=MaterialExpressionTextureSampleParameter2D
   inputs [
    UVs {from=MaterialExpressionAdd_1, output=""}
    Apply View MipBias {from="", output=""}
   ]
   properties {
    group=CustomOpacity
    parameterName=Custom Opacity - Texture
    samplerType=SAMPLERTYPE_LinearColor
    sortPriority=30
    texture=/Game/Global/Textures/Default/Noise/T_CloudNoise_01
   }
  }
  MaterialExpressionTime_0 {class=MaterialExpressionTime}
  MaterialExpressionTime_1 {class=MaterialExpressionTime}
  MaterialExpressionTransform_0 {
   class=MaterialExpressionTransform
   inputs [None {from=MaterialExpressionConstant3Vector_11}]
   properties {transformSourceType=TRANSFORMSOURCE_View}
  }
  MaterialExpressionTransform_1 {class=MaterialExpressionTransform, inputs [None {from=MaterialExpressionConstant3Vector_6}]}
  MaterialExpressionVectorParameter_0 {
   class=MaterialExpressionVectorParameter
   properties {group=Emissive, parameterName=Emissive - GlowColor, defaultValue=<rgba 4 0.7023 0.2042 1>}
  }
  MaterialExpressionVectorParameter_1 {
   class=MaterialExpressionVectorParameter
   properties {group=Emissive, parameterName=Emissive - Diffuse Color, defaultValue=<rgba 0.1327 0.1174 0.1083 1>}
  }
  MaterialExpressionVectorParameter_2 {class=MaterialExpressionVectorParameter, properties {group=Emissive, parameterName=Emissive - FresnelGlowColor, defaultValue=<rgba 0 0 0 1>}}
  MaterialExpressionVectorParameter_3 {
   class=MaterialExpressionVectorParameter
   properties {group=Fake Shadows, parameterName=Fake Shadows - GlowDirection, defaultValue=<rgba 0 0 1 1>}
  }
  MaterialExpressionVectorParameter_4 {
   class=MaterialExpressionVectorParameter
   properties {group=Fake Lighting, parameterName=Fake Lighting - Set Direction, sortPriority=2, defaultValue=<rgba 1 0 0 1>}
  }
  MaterialExpressionVectorParameter_5 {class=MaterialExpressionVectorParameter, properties {group=Fake Lighting, parameterName=Fake Lighting Color, defaultValue=<rgba 1 1 1 0>}}
  MaterialExpressionVectorParameter_6 {
   class=MaterialExpressionVectorParameter
   properties {group=Fake Lighting, parameterName=Fake Lighting - Noise Projection Size, defaultValue=<rgba 1024 1024 1024 1>}
  }
  MaterialExpressionVectorParameter_7 {
   class=MaterialExpressionVectorParameter
   properties {
    group=CustomOpacity
    parameterName=Custom Opacity - UVs
    channelNames {a=Panner Y, b=Panner X, g=Scale Y, r=Scale X}
    defaultValue=<rgba 1 1 0 0>
   }
  }
  MaterialExpressionVectorParameter_8 {
   class=MaterialExpressionVectorParameter
   properties {group=CustomOpacity, parameterName=Custom Opacity Channel Selector, defaultValue=<rgba 1 0 0 0>}
  }
  MaterialExpressionVertexNormalWS_0 {class=MaterialExpressionVertexNormalWS, inputs [], properties {}}
  MaterialExpressionVertexNormalWS_1 {class=MaterialExpressionVertexNormalWS, inputs [], properties {}}
  MaterialExpressionWorldPosition_0 {class=MaterialExpressionWorldPosition, inputs [], properties {}}
  MaterialExpressionWorldPosition_1 {class=MaterialExpressionWorldPosition, inputs [], properties {}}
  MaterialExpressionWorldPosition_2 {class=MaterialExpressionWorldPosition, inputs [], properties {}}
 }
 properties {
  bTangentSpaceNormal=false
  bUsedWithMeshParticles=true
  bUsedWithNanite=true
  bUsedWithNiagaraMeshParticles=true
  bUsedWithParticleSprites=true
  blendMode=BLEND_Translucent
  opacityMaskClipValue=0.9
  pixelDepthOffsetMode=PDOM_Legacy
  shadingModel=MSM_Unlit
  usedShadingModels=Unlit
 }
}
