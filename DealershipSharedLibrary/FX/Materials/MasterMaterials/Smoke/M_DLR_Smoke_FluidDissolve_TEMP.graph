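# Translucent unlit master smoke material for Niagara sprites in the dealership FX library, marked TEMP.
# Dynamic parameters drive SubUV texture dissolve through power and sharpness, plus a brightness transition on particle color.
# Emissive adds shadow/highlight tinting and optional Fresnel fake lighting blended toward sun/moon direction.
# Opacity uses particle alpha and depth fade, alongside an optional axis gradient and a noise-dissolve masked branch with optional temporal dithering.
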Material /DealershipSharedLibrary/FX/Materials/MasterMaterials/Smoke/M_DLR_Smoke_FluidDissolve_TEMP
graph {
 properties {
  bAutomaticallySetUsageInEditor=false
  bUsedWithNiagaraSprites=true
  blendMode=BLEND_Translucent
  shadingModel=MSM_Unlit
  usedShadingModels=Unlit
 }
 nodes [
  MaterialFunctionCall(Opacity In=NamedRerouteUsage(), DepthFade Distance=ScalarParameter() {defaultValue=1024, group="!Global Color", parameterName=Depth Fade}) {materialFunction=&MF_VCVFX_DepthFade}
  n1 = DynamicParameter() {paramNames [Dissolve, Sharpness, SmokeBrightness, Transition]}
  n2 = TextureSampleParameterSubUV!(UVs, Apply View MipBias) {
   group="!Global Color"
   parameterName=Particle Texture
   samplerType=SAMPLERTYPE_LinearColor
   sortPriority=-1
   texture=/DealershipSharedLibrary/FX/Textures/SubUVs/T_Smoke_HQ_N
  }
  n3 = ChannelMaskParameter(None=@n2.RGBA) {group="!Global Color", maskChannel=Alpha, parameterName=Alpha Texture Channel, sortPriority=2, defaultValue=<rgba 0 0 0 1>}
  n4 = Power(
   Base=If(
    A=ScalarParameter() {defaultValue=1, group="!Global Color", parameterName=Use Alpha Texture Channels, sortPriority=3}
    "A > B"=@n3.R
    "A == B"=@n3.R
    "A < B"=@n2.RGB
   ) {constB=1}
   Exp=@n1.Dissolve
  )
  NamedRerouteDeclaration(None=Saturate(None=Multiply(A=@n4, B=@n1.Sharpness))) {name=SubUV_Norm, nodeColor=<rgba 1 0.0938 0 1>}
  n5 = ChannelMaskParameter(None=@n2.RGBA) {group="!Global Color", parameterName=Emissive Texture Channel, sortPriority=2}
  n6 = If(
   A=ScalarParameter() {group="!Global Color", parameterName=Use Emissive Texture Channels, sortPriority=3}
   "A > B"=@n5.R
   "A == B"=@n5.R
   "A < B"=@n2.RGB
  ) {constB=1}
  n7 = ConstantBiasScale(None=DotProduct(A=@n6, B=Constant3Vector() {constant=<rgba 1 1 0 0>})) {bias=0.2}
  n8 = Power(
   Base=If(A=ScalarParameter() {defaultValue=1, group=Smoke Colors, parameterName=Use Texture Color}, "A > B"=@n6, "A == B"=@n7, "A < B"=@n7)
   Exp=ScalarParameter() {defaultValue=1, group="!Global Color", parameterName=Texture Card Highlight Amount}
  )
  n9 = ScalarParameter() {group="!Global Color", parameterName=Texture Card Highlight Bias}
  n10 = ParticleColor()
  n11 = Multiply(A=Desaturation(None=@n10.RGB, Fraction=Constant() {r=0.7}), B=@n1.SmokeBrightness)
  NamedRerouteDeclaration(None=Multiply(A=Add(A=@n8, B=@n9) {constB=0.2}, B=LinearInterpolate(A=@n10.RGB, B=@n11, Alpha=@n1.Transition))) {name=Emissive Card, nodeColor=<rgba 0 0.2266 1 1>}
  n12 = ScalarParameter() {defaultValue=1, group=Fake Lighting, parameterName=Use TOD vs Set Direction, sortPriority=2}
  n13 = CollectionParameter() {
   collection=&FortniteMaterialParameters
   parameterName=SunAndMoonModelDirectionalVector
  }
  n14 = StaticSwitchParameter(True=Transform(None=Constant3Vector() {constant=<rgba 0 0 1 0>}), False=CameraVectorWS()) {parameterName=Velocity Align Fresnel}
  n15 = StaticSwitchParameter(
   True=LinearInterpolate(
    A=ComponentMask(None=@n13) {b=true, g=true, r=true}
    B=@n14
    Alpha=ScalarParameter() {defaultValue=0.5, group=Fake Lighting, parameterName=Sun/Moon Directional Shading Blend Smoke}
   )
   False=@n14
  ) {defaultValue=true, group=Smoke Colors, parameterName=ToD Shading}
  n16 = VectorParameter() {group=Fake Lighting, parameterName=Fake Lighting Set Direction, sortPriority=2, defaultValue=<rgba 1 0 0 1>}
  n17 = Clamp(None=MaterialFunctionCall(
    Brightness=ScalarParameter() {defaultValue=1, group=Fake Lighting, parameterName=Main Shading Fresnel Brightness, sortPriority=2}
    Exponent=ScalarParameter() {defaultValue=2, group=Fake Lighting, parameterName=Main Shading Fresnel Exponent, sortPriority=2}
    Camera Vector=If(A=@n12, "A > B"=@n15, "A == B"=@n15, "A < B"=@n16.RGB) {constB=1}
   ) {materialFunction=&MF_Fresnel_FX}.Fresnel)
  n18 = Constant()
  NamedRerouteDeclaration(None=Subtract(A=NamedRerouteUsage(), B=StaticSwitchParameter(True=@n17, False=@n18) {group=Fake Lighting, parameterName=Use Fake Lighting, sortPriority=1})) {name=Lit Emissive Card, nodeColor=<rgba 0.5469 1 0 1>}
  n19 = ScalarParameter() {group=Smoke Colors, parameterName=Desaturate}
  n20 = ParticleColor()
  n21 = ScalarParameter() {defaultValue=1, group="!Global Color", parameterName=Particle Color - Shadow Brightness}
  n22 = NamedRerouteUsage()
  n23 = MaterialFunctionCall(In=@n22, Contrast=ScalarParameter() {group="!Global Color", parameterName=Smoke Emissive Contrast}) {materialFunction=&CheapContrast}
  n24 = VectorParameter() {group="!Global Color", parameterName=Color - Shadow, defaultValue=<rgba 0 0.1258 0.6667 1>}
  n25 = VectorParameter() {group="!Global Color", parameterName=Color - Highlight, defaultValue=<rgba 0 0.945 1 1>}
  n26 = Multiply(
   A=LinearInterpolate(A=@n20.RGB, B=Multiply(A=@n20.RGB, B=@n21), Alpha=@n23.Result)
   B=Desaturation(
    None=LinearInterpolate(A=@n24.RGB, B=@n25.RGB, Alpha=@n23.Result)
    Fraction=ScalarParameter() {defaultValue=1, group="!Global Color", parameterName=Particle Color - Use Particle Color}
   )
  )
  NamedRerouteDeclaration(None=MaterialFunctionCall(Emissive=If(A=@n19, "A > B"=@n22, "A == B"=@n26, "A < B"=@n26) {constB=1}) {materialFunction=&MF_FX_EmissiveAdjustments}.Result) {name=Emissive Color, nodeColor=<rgba 1 0 0.8672 1>}
  n27 = Divide(
   A=WorldPosition().XYZ
   B=Multiply(
    A=LocalPosition().XYZ
    B=VectorParameter() {group=Local Aligned Gradient, parameterName=Local Aligned Gradient - Axis, defaultValue=<rgba 0 0 1 0>}.RGB
   )
  )
  n28 = Divide(
   A=Add(
    A=ComponentMask(None=@n27) {b=true}
    B=ScalarParameter() {defaultValue=0.5, group=Local Aligned Gradient, parameterName=Local Aligned Gradient - Offset, sortPriority=2}
   )
   B=ScalarParameter() {defaultValue=2, group=Local Aligned Gradient, parameterName="Local Aligned Gradient -  Height", sortPriority=2}
  )
  NamedRerouteDeclaration(None=StaticSwitchParameter(True=Clamp(None=@n28), False=Constant() {r=1}) {group=Local Aligned Gradient, parameterName=Local Aligned Gradient, sortPriority=1}) {name=Local Aligned Gradient, nodeColor=<rgba 0 1 0.8125 1>}
  n29 = NamedRerouteUsage()
  n30 = NamedRerouteUsage()
  NamedRerouteDeclaration(None=Multiply(A=ParticleColor().A, B=Multiply(A=@n29, B=@n30))) {name=Opacity Out, nodeColor=<rgba 1 0.4922 0 1>}
  n31 = VectorParameter() {group=Masked, parameterName=Masked - Dissolve Tiling}
  n32 = Panner(Coordinate=Multiply(A=TextureCoordinate(), B=AppendVector(A=@n31.R, B=@n31.R)), Speed=AppendVector(A=@n31.B, B=@n31.B))
  n33 = TextureSampleParameter2D(UVs=@n32) {
   group=Masked
   parameterName="Dissolve  Tex"
   samplerType=SAMPLERTYPE_LinearColor
   texture=/Game/Global/Textures/Default/Tileable_Patterns/T_PerlinNoise_Uncompressed
  }
  n34 = Add(
   A=StaticSwitchParameter(True=@n33.RGB, False=TextureSample(UVs=@n32) {texture=/DealershipSharedLibrary/FX/Textures/Noise/T_VCVFX_clouds_noise_A}.RGB) {group=Fake Lighting, parameterName=Dissolve Tex Switch, sortPriority=1}
   B=DynamicParameter() {parameterIndex=1, defaultValue=<rgba -1 1 1 1>, paramNames [Masked Dissolve, NA, NA, NA]}.Masked Dissolve
  )
  NamedRerouteDeclaration(None=Saturate(None=OneMinus(None=@n34))) {name=Normalized Gradient for Mask, nodeColor=<rgba 0.1719 0 1 1>}
  n35 = MaterialFunctionCall(
   In=Multiply(A=NamedRerouteUsage(), B=NamedRerouteUsage())
   Contrast=ScalarParameter() {defaultValue=2, group=Masked, parameterName=Masked - Opacity Contrast}
  ) {materialFunction=&CheapContrast}
  n36 = MaterialFunctionCall(Alpha Threshold=@n35.Result) {materialFunction=&DitherTemporalAA}
  NamedRerouteDeclaration(None=If(
    A=ScalarParameter() {group=Masked, parameterName=Masked - Dithered Temporal AA}
    "A > B"=@n36.Result
    "A == B"=@n36.Result
    "A < B"=@n35.Result
   ) {constB=1}) {name=Masked Opacity, nodeColor=<rgba 0.1484 1 0 1>}
  NamedRerouteUsage()
  NamedRerouteUsage()
 ]
}
