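# Translucent unlit master material for snow mist and smoke particles on Niagara sprites, ribbons and meshes.
# Emissive combines channel-selected, optionally desaturated texture and particle color with fake three-point lighting from RGB channels and sun/moon or set direction, or switches to a stylized color path.
# Opacity multiplies texture, particle alpha and mask, optionally hardened by a Subtract/Ceil cutoff, then depth-faded.
# Dynamic parameters drive brightness, cutoff and light color blend.
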Material /DealershipSharedLibrary/FX/Materials/MasterMaterials/Smoke/M_DLR_SnowMist
graph {
 properties {
  bAutomaticallySetUsageInEditor=false
  bUsedWithNiagaraMeshParticles=true
  bUsedWithNiagaraRibbons=true
  bUsedWithNiagaraSprites=true
  blendMode=BLEND_Translucent
  shadingModel=MSM_Unlit
  translucencyLightingMode=TLM_VolumetricPerVertexDirectional
  usedShadingModels=Unlit
 }
 nodes [
  n1 = ComponentMask(None=NamedRerouteUsage()) {b=true, g=true, r=true}
  n2 = Multiply(A=NamedRerouteUsage(), B=ScalarParameter() {group=3 Point Lighting, parameterName=3PointLighting_Intensity, sortPriority=1})
  n3 = NamedRerouteUsage()
  n4 = TextureSampleParameter2D() {parameterName=MainText, texture=/DealershipSharedLibrary/FX/Textures/Flipbook/T_VCVFX_SmokeHQ_N}
  n5 = If(A=ScalarParameter() {parameterName=ClampMainTexture?}, "A > B"=@n3, "A == B"=@n3, "A < B"=@n4.RGB) {constB=1}
  n6 = VectorParameter() {group=Color Blending, parameterName=Base Color Texture Channel, sortPriority=1, defaultValue=<rgba 1 0 0 0>}
  n7 = DotProduct(A=@n5, B=@n6.RGBA)
  n8 = If(
   A=ScalarParameter() {group=Color Blending, parameterName=Use Texture Channel - Base Color, sortPriority=1}
   "A > B"=@n7
   "A == B"=@n7
   "A < B"=@n5
  ) {constB=1}
  n9 = Desaturation(None=@n8)
  n10 = If(A=ScalarParameter() {defaultValue=1, parameterName=UseDesaturate?}, "A > B"=@n9, "A == B"=@n9, "A < B"=@n8) {constB=1}
  MaterialExpressionVertexColor_0 = ParticleColor()
  n11 = DynamicParameter() {paramNames [Mask, NoiseAmount, ColorTransition, Brightness]}
  n12 = Multiply(
   A=Multiply(A=Multiply(A=@n10, B=@MaterialExpressionVertexColor_0.RGB), B=@n11.Brightness)
   B=ScalarParameter() {defaultValue=1, group=Color Blending, parameterName=Base Color Emissive Value, sortPriority=1}
  )
  MaterialFunctionCall(Emissive=If(A=ScalarParameter() {parameterName=UseStylizedLerpColor?}, "A > B"=@n1, "A == B"=@n1, "A < B"=Add(A=@n2, B=@n12)) {constB=1}) {materialFunction=&MF_FX_EmissiveAdjustments}
  n13 = VectorParameter() {group=Color Blending, parameterName=Opacity Texture Channel, sortPriority=2, defaultValue=<rgba 0 0 1 0>}
  n14 = DotProduct(A=@n4.RGBA, B=@n13.RGBA)
  n15 = If(
   A=ScalarParameter() {group=Color Blending, parameterName=Use Texture Channel - Opacity, sortPriority=2}
   "A > B"=@n14
   "A == B"=@n14
   "A < B"=@n4.A
  ) {constB=1}
  n16 = NamedRerouteUsage()
  MaterialExpressionConstantClamp_0 = Saturate(None=Multiply(A=@MaterialExpressionVertexColor_0.A, B=Multiply(A=@n15, B=@n16)))
  n17 = Multiply(
   A=@MaterialExpressionConstantClamp_0
   B=TextureSampleParameter2D(UVs=TextureCoordinate()) {parameterName=OpacityTextureMask, samplerType=SAMPLERTYPE_LinearColor, texture=/Bridge/MSPresets/MSTextures/White}.R
  )
  n18 = Subtract(A=@n17, B=DynamicParameter() {parameterIndex=1, defaultValue=<rgba 0 1 1 1>, paramNames [Subtract,Param2,Param3,Param4]}.Subtract) {constB=0}
  n19 = ScalarParameter() {defaultValue=0.4, parameterName=AddToClapmedOpacity Mask}
  n20 = MaterialFunctionCall(Input=Add(A=Ceil(None=@n18), B=@n19) {constB=0.4}, Input Low=@n19, Input High=Add(A=@n19)) {materialFunction=&RemapValue_0-1}
  MaterialFunctionCall(
   Opacity In=If(A=ScalarParameter() {parameterName=ClampOpacityMask?}, "A > B"=@n20.Result, "A == B"=@n20.Result, "A < B"=@n17) {constB=1}
   DepthFade Distance=ScalarParameter() {defaultValue=100, group=Opacity, parameterName=DepthFade-Standard_Distance, sortPriority=0}
  ) {materialFunction=&MF_VCVFX_DepthFade}
  n21 = TextureSampleParameter2D() {parameterName=MainText, texture=&T_VCVFX_SmokeHQ_N}
  n22 = CollectionParameter() {
   collection=&FortniteMaterialParameters
   parameterName=SunAndMoonModelDirectionalVector
  }
  n23 = ComponentMask(None=@n22) {b=true, g=true, r=true}
  n24 = VectorParameter() {group=3 Point Lighting, parameterName=Fake Lighting Set Direction, sortPriority=2, defaultValue=<rgba 1 0 0 1>}
  n25 = If(
   A=ScalarParameter() {defaultValue=1, group=3 Point Lighting, parameterName=Use TOD vs Set Direction, sortPriority=2}
   "A > B"=@n23
   "A == B"=@n23
   "A < B"=Transform(None=@n24.RGB)
  ) {constB=1}
  n26 = LinearInterpolate(
   A=@n25
   B=DotProduct(A=@n25, B=CameraVectorWS())
   Alpha=ScalarParameter() {defaultValue=0.5, group=3 Point Lighting, parameterName=Blend World Direction with Camera Direction, sortPriority=2}
  )
  n27 = LinearInterpolate(A=@n21.R, B=Constant(), Alpha=ComponentMask(None=@n26) {r=true})
  n28 = LinearInterpolate(A=@n21.G, B=Constant(), Alpha=ComponentMask(None=@n26) {g=true})
  n29 = Constant()
  n30 = ComponentMask(None=@n26) {b=true}
  n31 = Power(
   Base=Add(A=Add(A=@n27, B=@n28), B=LinearInterpolate(A=@n21.B, B=@n29, Alpha=@n30))
   Exp=ScalarParameter() {defaultValue=5, group=3 Point Lighting, parameterName=3 Point Lighting Exp}
  )
  n32 = LinearInterpolate(
   A=VectorParameter() {parameterName=3PointLighting_Color1, defaultValue=<rgba 1 1 1 0>}.RGB
   B=VectorParameter() {parameterName=3PointLighting_Color2, defaultValue=<rgba 1 1 1 0>}.RGB
   Alpha=DynamicParameter() {parameterIndex=1, defaultValue=<rgba 0 1 1 1>, paramNames [Subtract,Param2,Param3,Param4]}.Param2
  )
  NamedRerouteDeclaration(None=Multiply(
    A=Multiply(A=@n31, B=@n32)
    B=DynamicParameter() {parameterIndex=1, defaultValue=<rgba 0 1 1 1>, paramNames [Subtract,Param2,Param3,Param4]}.Subtract
   )) {name=3Point_Lighting, nodeColor=<rgba 1 0.4922 0 1>}
  n33 = VectorParameter() {parameterName=ColorLerpA, defaultValue=<rgba 0.169 0.3 0.1944 0.8>}
  n34 = VectorParameter() {parameterName=ColorLerpB, defaultValue=<rgba 0.0375 0.1 0.095 0.7>}
  n35 = NamedRerouteUsage()
  n36 = LinearInterpolate(A=@n33.A, B=@n34.A, Alpha=@n35)
  n37 = TextureSampleParameter2D() {parameterName=MainText, texture=&T_VCVFX_SmokeHQ_N}
  n38 = Saturate(None=Multiply(A=Subtract(A=@n37.A) {constB=0.3}) {constB=20})
  NamedRerouteDeclaration(None=LinearInterpolate(A=SceneColor!(UVs).RGB, B=LinearInterpolate(A=@n33.RGB, B=@n34.RGB, Alpha=@n35), Alpha=Multiply(A=@n36, B=@n38))) {name=LerpColor, nodeColor=<rgba 1 0 0.1406 1>}
  NamedRerouteDeclaration(None=@n11.Mask) {name=DP_Mask, nodeColor=<rgba 0.1355 0.4006 1 1>}
  n39 = NamedRerouteUsage()
  n40 = DotProduct(A=Transform(None=Multiply(A=@n39, B=@n4.RGBA)), B=Constant3Vector() {constant=<rgba 0 0 1 0>})
  NamedRerouteDeclaration(None=Saturate(None=Ceil(None=@n40))) {name=CeilTexture, nodeColor=<rgba 0.4297 1 0 1>}
  NamedRerouteDeclaration(None=@n38) {name=alphaMask, nodeColor=<rgba 1 0.1755 0.8588 1>}
  Reroute(None=TextureSampleParameter2D() {parameterName=NormalTex, texture=&T_VCVFX_SmokeHQ_N}.RGB)
 ]
}
