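# Translucent unlit master material for Niagara smoke particles, with optional HDR glitter and scrolling noise dissolve.
# Glitter multiplies one noise texture sampled at two tilings, tints and boosts it, then feeds both emissive and opacity.
# Particle color modulates the smoke flipbook, and emissive passes through MF_FX_EmissiveAdjustments.
# A dynamic parameter drives dissolve before depth fade; with dissolve off (default), opacity is multiplied by itself.
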Material /DealershipSharedLibrary/FX/Materials/MasterMaterials/Smoke/M_Smoke_Glitter
graph {
 properties {
  bAutomaticallySetUsageInEditor=false
  bUsedWithNiagaraMeshParticles=true
  bUsedWithNiagaraRibbons=true
  bUsedWithNiagaraSprites=true
  blendMode=BLEND_Translucent
  shadingModel=MSM_Unlit
  usedShadingModels=Unlit
 }
 nodes [
  n1 = VectorParameter() {parameterName=GlitterColor, defaultValue=<rgba 1 0.4176 0 1>}
  n2 = TextureCoordinate() {uTiling=10, vTiling=10}
  n3 = Constant() {r=0.5}
  n4 = Multiply(
   A=TextureSampleParameter2D(UVs=@n2) {parameterName=GlitterTexture, texture=/DealershipSharedLibrary/FX/Textures/Noise/T_VCVFX_NoiseGeneric01Pack_RGB}.B
   B=TextureSampleParameter2D(UVs=Multiply(A=@n2, B=@n3)) {parameterName=GlitterTexture, texture=&T_VCVFX_NoiseGeneric01Pack_RGB}.B
  )
  n5 = Saturate(None=Multiply(A=Desaturation(None=@n4)) {constB=10})
  n6 = Multiply(A=Multiply(A=Multiply(A=@n1.RGB, B=@n5), B=Constant() {r=1024}), B=ScalarParameter() {defaultValue=1, parameterName=GlitterGlow})
  n7 = ParticleColor()
  n8 = TextureSampleParameter2D() {parameterName=MainTexture, texture=/DealershipSharedLibrary/FX/Textures/Flipbook/T_VCVFX_smoke_puff_8x8}
  n9 = Multiply(A=@n7.RGB, B=Multiply(A=@n8.B, B=ScalarParameter() {defaultValue=1, group=Main Texture, parameterName=Texture Multiplier}))
  MaterialFunctionCall(Emissive=Add(A=StaticSwitchParameter(True=@n6, False=@n9) {defaultValue=true, parameterName=Use Glitter on/off}, B=@n9)) {materialFunction=&MF_FX_EmissiveAdjustments}
  n10 = LinearInterpolate(A=@n8.A, B=@n8.A, Alpha=ScalarParameter() {defaultValue=1, parameterName=Alpha/RGB})
  n11 = Multiply(A=StaticSwitchParameter(True=@n6, False=Constant()) {defaultValue=true, parameterName=Use Glitter on/off}, B=@n10)
  n12 = Constant() {r=1}
  n13 = Add(A=Saturate(None=Desaturation(None=@n11)), B=Multiply(A=@n7.A, B=Multiply(A=@n10, B=@n12)))
  n14 = TextureCoordinate()
  n15 = AppendVector(
   A=ScalarParameter() {defaultValue=1, group=Dissolve, parameterName=Tilinng X, sortPriority=1}
   B=ScalarParameter() {defaultValue=1, group=Dissolve, parameterName=Tiling Y, sortPriority=1}
  )
  n16 = AppendVector(
   A=ScalarParameter() {defaultValue=1, group=Dissolve, parameterName=Scroll X, sortPriority=2}
   B=ScalarParameter() {group=Dissolve, parameterName=Scroll Y, sortPriority=2}
  )
  n17 = TextureSampleParameter2D(UVs=Add(A=Multiply(A=@n14, B=@n15), B=Panner(Coordinate=@n14, Speed=@n16))) {group=Dissolve, parameterName=Dissolve Texture, sortPriority=0, texture=/Game/Effects/Fort_Effects/Textures/Fire/T_FireNoise_02}
  n18 = Saturate(None=Subtract(A=@n17.RGB, B=DynamicParameter() {paramNames [NoiseTransition, GlitterGlow, Dissolve, Param4]}.Dissolve))
  MaterialFunctionCall(
   Opacity In=Multiply(A=@n13, B=StaticSwitchParameter(True=@n18, False=@n13) {parameterName=Use Dissolve on/off})
   DepthFade Distance=ScalarParameter() {defaultValue=50, group=Depth Fade, parameterName=DepthFade Distance}
  ) {materialFunction=&MF_VCVFX_DepthFade}
 ]
}
